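; Destructible event pillar that players damage, then resolves as winner or loser.
; Clients poll health on a timer to drive progress and crack material parameters; hits play impact sounds and shatter bursts.
; IsWinner runs timelines: winners crack, launch a spinning, shrinking gun, and trigger Zero Point destabilization and camera shake; losers sink.
; The server records the outcome in replicated PillarState, which restores transforms during replays, including after scrubbing stops the timelines.
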(event EventBeginPlay
  (|CallParentFunction)
  (Networking|SwitchHasAuthority
    (:Authority
      (Variables|Default|SetPillarState 0))
    (:Remote
      (Utilities|FlowControl|Sequence
        (:then_0
          (Utilities|FlowControl|Delay (Math|Random|RandomFloatinRange 0.0 1.0))
          (CallFunction|StartMaterialTimer))
        (:then_1
          (if (Replay|GetIsPlayingReplay)
            (ReplayContext|BindEventtoOnReplayScrubComplete (BlueprintContext|GetFortReplayContext) (EventDispatchers|CreateEvent :!event Custom|ResetSequenceDoonce))))))))

(event Custom|StartMaterialTimer
  (bind _returnvalue (Utilities|Time|SetTimerbyEvent (delegate Custom|UpdateHealthMaterial) 0.1 true))
  (Variables|Default|SetUpdateMaterialTimer _returnvalue))

(event Custom|UpdateHealthMaterial
  (bind _pillar (Variables|Default|GetPillar))
  (bind _output_get (Variables|Default|SetRatio (- 1.0 (CallFunction|GetHealthValuePercent))))
  (Rendering|Material|SetScalarParameterValueonMaterials _pillar "ProgressBar" _output_get)
  (Rendering|Material|SetScalarParameterValueonMaterials _pillar "Animation" (Math|Float|MapRangeClamped (Math|Float|Power (Variables|Default|GetRatio) (Variables|Default|GetPow)) 0.0 1.0 0.0 0.25)))

(event Collision|EventActorBeginOverlap
  (params (OtherActor :type Actor%20Object%20Reference))
  (|CallParentFunction OtherActor) ; This node is disabled and will not be called.\nDrag off pins to build functionality.
)

(event EventTick
  (params (DeltaSeconds :type Float%20%28single-precision%29))
  (|CallParentFunction DeltaSeconds) ; This node is disabled and will not be called.\nDrag off pins to build functionality.
)

(event EventDirtOutOfHealth)

(event Custom|RiseUp
  (bind (_direction _newtrack_0) (Timeline|Timeline_2 PlayFromStart)
    (:Update
      (Transformation|SetRelativeLocation (Variables|Default|GetPillar) (Math|Vector|Lerp(Vector) "0, 0,-1280.000000" "0, 0, 0" _newtrack_0) false false))))

(event Custom|ResetDoonce
  ;[1 1 1]
  (Utilities|FlowControl|DoOnce false
    (:Completed
      (Components|Activation|Activate (Variables|Default|GetP_DirtShell_Finale) false)
      (Development|PrintString "PILLARSNDPLS"))))

(event Custom|IsWinner
  (params (Winner :type Boolean))
  (bind _zeropoint (Variables|Default|GetZeroPoint))
  (bind _scene (Variables|Default|GetScene))
  (Networking|FlushNetDormancy)
  (if Winner
    (CallFunction|OnPillarBroken)
    (Utilities|FlowControl|Delay 3.0)
    (bind (_direction _crackpillar _gun_y_axis _gun_rotation _gun_z) (Timeline|WinnerTimeline PlayFromStart)
      (:Update
        (Utilities|FlowControl|Sequence
          (:then_0
            ;[1 1 1]
            (Utilities|FlowControl|DoOnce false
              (:Completed
                (Components|Activation|Activate (Variables|Default|GetP_DirtShell_Finale) false)
                (Development|PrintString "PILLARSNDPLS"))))
          (:then_1
            (Rendering|Material|SetScalarParameterValueonMaterials (Variables|Default|GetPillar) "animation" (Math|Float|Lerp 0.25 1.0 _crackpillar))
            (bind (_returnvalue_x _returnvalue_y) (Math|Vector|BreakVector (Math|Vector|Lerp(Vector) "0, 0, 0" "0,300.000000, 0" _gun_y_axis)))
            (Transformation|SetRelativeLocationAndRotation _scene (Math|Vector|MakeVector 0.0 _returnvalue_y (Math|Float|Lerp 0.0 8e+03 _gun_z)) (Math|Rotator|Lerp(Rotator) "0, 0, 0" "0,1800.000000, 0" _gun_rotation) false false)
            (Transformation|SetRelativeScale3D _scene (Math|Vector|MakeVector (Math|Float|Lerp 1.5 0.001 _gun_z) (Math|Float|Lerp 1.5 0.001 _gun_z) 1.5))
            (Utilities|FlowControl|DoOnce false
              (:Completed
                (Utilities|Time|ClearandInvalidateTimerbyHandle (Variables|Default|GetUpdateMaterialTimer)))))))
      (:Finished
        (CallFunction|ResetDoonce)
        (Networking|SwitchHasAuthority
          (:Authority
            (Variables|Default|SetPillarState 2))))
      (:GunTeleport
        (Components|Activation|Activate (Variables|Default|GetP_DirtShell_Teleport) false))
      (:GunFloatingParticles
        (Components|Activation|ToggleActive (Variables|Default|GetP_DirtShell_Wisps)))
      (:TriggerDestable
        (Class|BPZreoPointDestabilize|CETriggerEffect _zeropoint)
        (Camera|PlayWorldCameraShake "/Game/Effects/Fort_Effects/Camera_Effects/B_Generic_Quake_BuildUp_CameraShake.B_Generic_Quake_BuildUp_CameraShake_C" (Transformation|GetActorLocation _zeropoint) 1e+04 2.5e+03)))
    (else
      (Utilities|FlowControl|Delay (Math|Random|RandomFloatinRange 0.0 2.0))
      (bind (_direction_2 _move _fadeao) (Timeline|LoserTimeline PlayFromStart)
        (:Update
          (Rendering|Material|SetScalarParameterValueonMaterials (Variables|Default|GetFakeAO) "Opacity" _fadeao)
          (Transformation|SetRelativeLocation (Variables|Default|GetPillar) (Math|Vector|Lerp(Vector) "0, 0, 0" "0, 0,-1600.000000" _move) false false))
        (:Finished
          (Networking|SwitchHasAuthority
            (:Authority
              (Variables|Default|SetPillarState 1))))
        (:NewTrack_0
          (Audio|PlaySoundatLocation "/Figment_S08/DirtShell_8_5_1/Athena/Sounds/Events/Dopey/Snow/Snow_Dest_Amb_Descend_2_Cue.Snow_Dest_Amb_Descend_2_Cue" (Transformation|GetWorldLocation (Variables|Default|GetAudio))))))))

(event Custom|Winner
  (params (Condition :type Boolean))
  (bind _zeropoint (Variables|Default|GetZeroPoint))
  (bind _scene (Variables|Default|GetScene))
  (if (stale Custom|IsWinner Winner)
    (CallFunction|OnPillarBroken)
    (Utilities|FlowControl|Delay 3.0)
    (Timeline|WinnerTimeline PlayFromStart)
    (else
      (Utilities|FlowControl|Delay (Math|Random|RandomFloatinRange 0.0 2.0))
      (Timeline|LoserTimeline PlayFromStart))))

(event Custom|PauseOnScrub
  (bind _zeropoint (Variables|Default|GetZeroPoint))
  (bind _scene (Variables|Default|GetScene))
  (Timeline|WinnerTimeline Stop))

(event Custom|On%20Pillar%20Broken
  (Networking|CanExecuteCosmeticEvents
    (:True
      (bind _returnvalue (Audio|SpawnSoundatLocation "/Figment_S08/DirtShell_8_5_1/Athena/Sounds/Events/Dopey/Snow/Snow_Dest_Amb_Winner_2_Cue.Snow_Dest_Amb_Winner_2_Cue" (Transformation|GetWorldLocation (Variables|Default|GetAudio))))
      (Variables|Default|SetVictorySound _returnvalue))))

(event Custom|ResetSequenceDoonce
  (Utilities|FlowControl|Sequence
    (:then_0
      (Utilities|FlowControl|DoOnce false
        (:Completed
          (Utilities|Time|ClearandInvalidateTimerbyHandle (Variables|Default|GetUpdateMaterialTimer)))))
    (:then_1
      (CallFunction|PauseonScrub)
      (CallFunction|StopLosersonScrub)
      (CallFunction|DealwithPillarState))))

(event Custom|StopLosersOnScrub
  (Timeline|LoserTimeline Stop))

(event EventOnDamagePlayEffects
  (params (Damage :type Float%20%28single-precision%29) (DamageTags :type Gameplay%20Tag%20Container%20Structure%20%28by%20ref%29) (Momentum :type Vector :default "0, 0, 0") (HitInfo :type Hit%20Result%20Structure%20%28by%20ref%29) (InstigatedBy :type Fort%20Pawn%20Object%20Reference) (DamageCauser :type Actor%20Object%20Reference) (EffectContext :type Gameplay%20Effect%20Context%20Handle%20Structure))
  (bind (_bblockinghit _binitialoverlap _time _distance _location) (Collision|BreakHitResult HitInfo))
  (bind _p_dirtshell_eachshatterburst (Variables|Default|GetP_DirtShell_EachShatterBurst))
  (|CallParentFunction Damage DamageTags Momentum HitInfo InstigatedBy DamageCauser EffectContext)
  (Audio|PlaySoundatLocation "/Figment_S08/DirtShell_8_5_1/Athena/Sounds/Events/Dopey/Snow/Snow_Impact_Cue.Snow_Impact_Cue" _location)
  (Audio|PlaySoundatLocation "/Figment_S08/DirtShell_8_5_1/Athena/Sounds/Events/Dopey/Snow/Snow_Impact_Resonant.Snow_Impact_Resonant" _location)
  (Components|Activation|Activate _p_dirtshell_eachshatterburst true)
  (Transformation|SetWorldLocation _p_dirtshell_eachshatterburst _location false false))

(event Custom|DealWithPillarState
  (bind _scene (Variables|Default|GetScene))
  (bind _scene_2 (Variables|Default|GetScene))
  ;[1 1 1] Rep pillar state
  (if (Replay|GetIsPlayingReplay)
    (switch int (Variables|Default|GetPillarState)
      (:0
        (Transformation|SetRelativeLocation (Variables|Default|GetPillar) "0, 0, 0" false false)
        (CallFunction|StartMaterialTimer)
        (Transformation|SetRelativeLocationAndRotation _scene (0.0 0.0 0.0) "0, 0, 0" false false)
        (Transformation|SetRelativeScale3D _scene (1.5 1.5 1.5)))
      (:1
        (Transformation|SetRelativeLocation (Variables|Default|GetPillar) "0, 0,-1600.000000" false false))
      (:2
        (Transformation|SetRelativeLocationAndRotation _scene_2 (0.0 3e+02 8e+03) "0,1800.000000, 0" false false)
        (Transformation|SetRelativeScale3D _scene_2 (0.001 0.001 1.5))))))

(fn OnRep_PillarState
  (CallFunction|DealwithPillarState))

(fn OnRep_WInner)

(fn ConstructionScript
  (|CallParentFunction))
