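; Child dirt-shell actor that raises a Fill actor once all dirt mounds blocking a vent are destroyed.
; On the server, it counts prop deaths, including props already gone at BeginPlay, then raises Fill by timeline if live or instantly during setup.
; A box trace below Fill destroys tiered chests.
; For live triggers, replicated TriggeredInGame makes clients play an earthquake cue and audio; debug prints and a test trigger remain.
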(event EventBeginPlay
  (|CallParentFunction)
  (Networking|SwitchHasAuthority
    (:Authority
      (Utilities|FlowControl|Sequence
        (:then_0
          ;[1 1 1] Check to see if the dirt mounds blocking the vent have already been destroyed
          (for _array_element (Variables|Default|GetDirtShellProps)
            (Utilities|IsValid _array_element
              (:Is%20Valid
                (EventDispatchers|BindEventtoOnDied _array_element (EventDispatchers|CreateEvent :!event Custom|DeathEvent)))
              (:Is%20Not%20Valid
                (CallFunction|PropDied)))))
        (:then_1
          (if (== (Variables|Default|GetNumDead) (Variables|Default|GetMaxNumDirt))
            (CallFunction|EventTrigger)))))))

(event Collision|EventActorBeginOverlap
  (params (OtherActor :type Actor%20Object%20Reference))
  (|CallParentFunction OtherActor) ; This node is disabled and will not be called.\nDrag off pins to build functionality.
)

(event EventTick
  (params (DeltaSeconds :type Float%20%28single-precision%29))
  (|CallParentFunction DeltaSeconds) ; This node is disabled and will not be called.\nDrag off pins to build functionality.
)

(event Custom|PropDied
  (Math|Integer|IncrementInt (Variables|Default|GetNumDead)
    (:%20%20
      (if (== (Variables|Default|GetNumDead) (Variables|Default|GetMaxNumDirt))
        (if (Variables|Default|GetFirstSetupComplete)
          (Utilities|FlowControl|Sequence
            (:then_0
              (CallFunction|EventTrigger))
            (:then_1
              (Variables|Default|SetTriggeredInGame true)
              (Utilities|FlowControl|Delay 4.0)
              (Variables|Default|SetTriggeredInGame false)))
          (else
            (CallFunction|EventTrigger)))))))

(event Custom|DeathEvent
  (params (DamagedActor :type Actor%20Object%20Reference) (Damage :type Float%20%28single-precision%29) (InstigatedBy :type Controller%20Object%20Reference) (DamageCauser :type Actor%20Object%20Reference) (HitLocation :type Vector) (FHitComponent :type Primitive%20Component%20Object%20Reference) (BoneName :type Name) (Momentum :type Vector))
  (CallFunction|PropDied))

(event Custom|TestEventTrigger
  (CallFunction|TriggerQuake)
  (Networking|FlushNetDormancy)
  (Networking|SwitchHasAuthority
    (:Authority
      (Variables|Default|SetFillOriginLocation (Transformation|GetActorLocation (Variables|Default|GetFill)))
      (Rendering|SetActorHiddenInGame (Variables|Default|GetFill) false)
      (if (Variables|Default|GetFirstSetupComplete)
        (Development|PrintString "this was triggered live" true true "(R=0.000000,G=0.660000,B=1.000000,A=1.000000)" 2e+01)
        (bind (_direction _raise) (Timeline|Raise PlayFromStart)
          (:Update
            (Transformation|SetActorLocation (Variables|Default|GetFill) (+ (Variables|Default|GetFillOriginLocation) (Math|Vector|MakeVector 0.0 0.0 (* _raise (Variables|Default|GetOffsetFill)))) false false))
          (:Finished
            _))
        (else
          (Development|PrintString "this was triggered at begin play" true true "(R=0.000000,G=0.660000,B=1.000000,A=1.000000)" 2e+01)
          (Transformation|SetActorLocation (Variables|Default|GetFill) (+ (Variables|Default|GetFillOriginLocation) (Math|Vector|MakeVector 0.0 0.0 (Variables|Default|GetOffsetFill))) false false)))
      (bind _outhits (Collision|MultiBoxTraceForObjects :Start (- (Transformation|GetActorLocation (Variables|Default|GetFill)) "0, 0,1000.000000") :End (- (Transformation|GetActorLocation (Variables|Default|GetFill)) "0, 0,1000.000000") :HalfSize "2000.000000,2000.000000,1000.000000" :Orientation "0, 0, 0" :ObjectTypes (Utilities|Array|MakeArray "ObjectTypeQuery1") :bTraceComplex false :DrawDebugType "ForDuration" :bIgnoreSelf true :TraceColor "(R=1.000000,G=0.000000,B=0.000000,A=1.000000)" :TraceHitColor "(R=0.000000,G=1.000000,B=0.000000,A=1.000000)" :DrawTime 6e+01))
      (for _array_element _outhits
        (bind _astiered_chest_athena (Utilities|Casting|CastToTiered_Chest_Athena (.HitActor (Collision|BreakHitResult _array_element))))
        (Development|PrintString (Utilities|String|Append (Utilities|GetDisplayName (.HitActor (Collision|BreakHitResult _array_element))) " is an object") true true "(R=0.000000,G=0.660000,B=1.000000,A=1.000000)" 6e+01)
        (Actor|DestroyActor _astiered_chest_athena)))))

(event Custom|EventTrigger
  (Networking|SwitchHasAuthority
    (:Authority
      (Variables|Default|SetFillOriginLocation (Transformation|GetActorLocation (Variables|Default|GetFill)))
      (Rendering|SetActorHiddenInGame (Variables|Default|GetFill) false)
      (if (Variables|Default|GetFirstSetupComplete)
        (Development|PrintString "this was triggered live" true true "(R=0.000000,G=0.660000,B=1.000000,A=1.000000)" 2e+01)
        (Timeline|Raise PlayFromStart)
        (else
          (Development|PrintString "this was triggered at begin play" true true "(R=0.000000,G=0.660000,B=1.000000,A=1.000000)" 2e+01)
          (Transformation|SetActorLocation (Variables|Default|GetFill) (+ (Variables|Default|GetFillOriginLocation) (Math|Vector|MakeVector 0.0 0.0 (Variables|Default|GetOffsetFill))) false false)))
      (bind _outhits (Collision|MultiBoxTraceForObjects :Start (- (Transformation|GetActorLocation (Variables|Default|GetFill)) "0, 0,1000.000000") :End (- (Transformation|GetActorLocation (Variables|Default|GetFill)) "0, 0,1000.000000") :HalfSize "2000.000000,2000.000000,1000.000000" :Orientation "0, 0, 0" :ObjectTypes (Utilities|Array|MakeArray "ObjectTypeQuery1") :bTraceComplex false :DrawDebugType "ForDuration" :bIgnoreSelf true :TraceColor "(R=1.000000,G=0.000000,B=0.000000,A=1.000000)" :TraceHitColor "(R=0.000000,G=1.000000,B=0.000000,A=1.000000)" :DrawTime 6e+01))
      (for _array_element _outhits
        (bind _astiered_chest_athena (Utilities|Casting|CastToTiered_Chest_Athena (.HitActor (Collision|BreakHitResult _array_element))))
        (Development|PrintString (Utilities|String|Append (Utilities|GetDisplayName (.HitActor (Collision|BreakHitResult _array_element))) " is an object") true true "(R=0.000000,G=0.660000,B=1.000000,A=1.000000)" 6e+01)
        (Actor|DestroyActor _astiered_chest_athena)))))

(event Custom|TriggerQuake
  (Networking|SwitchHasAuthority
    (:Remote
      (GameplayCue|ExecuteGameplayCue%28Burst%29Local (Pawn|GetControlledPawn (Utilities|Array|Get(acopy) (Fortnite|GetLocalFortPlayerControllers true true))) (Variables|Default|GetCueTag_EarthquakeTremors) (Ability|GameplayCue|MakeGameplayCueParameters :NormalizedMagnitude 0.0 :RawMagnitude 0.0 :EffectContext "()" :AggregatedSourceTags "(GameplayTags=)" :AggregatedTargetTags "(GameplayTags=)" :Location "0,0,0" :Normal "0,0,0" :EffectCauser self :GameplayEffectLevel 1 :AbilityLevel 1 :bReplicateLocationWhenUsingMinimalRepProxy false))
      (Audio|Components|Audio|Play (Variables|/Figment_S08/DirtShell_8_5_1/Athena/Prototype/Blueprints/Dopey/BP_Dopey_Fill_B_8_5_1.BP_Dopey_Fill_B_8_5_1_C|GetAudio (Variables|Default|GetFill)) 0.0))))

(fn OnRep_TriggeredInGame
  (if (Variables|Default|GetTriggeredInGame)
    (CallFunction|TriggerQuake)))

(fn ConstructionScript
  (|CallParentFunction))
