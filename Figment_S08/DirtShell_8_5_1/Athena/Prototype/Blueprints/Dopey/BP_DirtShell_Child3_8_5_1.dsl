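; BP_DirtShell_Child3_8_5_1 is a prototype dirt shell child Blueprint that fires an earthquake tremor effect when its dig activation replicates.
; OnRep_ActivatedOnDig runs only on non-authority instances, printing a debug string and executing a local burst GameplayCue with the EarthquakeTremors tag.
; The cue targets the first local player controller's pawn, with this actor as effect causer.
; BeginPlay and construction defer to the parent, while overlap and tick are disabled stubs.
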(event EventBeginPlay
  (|CallParentFunction))

(event Collision|EventActorBeginOverlap
  (params (OtherActor :type Actor%20Object%20Reference))
  (|CallParentFunction OtherActor) ; This node is disabled and will not be called.\nDrag off pins to build functionality.
)

(event EventTick
  (params (DeltaSeconds :type Float%20%28single-precision%29))
  (|CallParentFunction DeltaSeconds) ; This node is disabled and will not be called.\nDrag off pins to build functionality.
)

(fn OnRep_ActivatedOnDig
  (Networking|SwitchHasAuthority
    (:Remote
      (Development|PrintString "called GC")
      (GameplayCue|ExecuteGameplayCue%28Burst%29Local (Pawn|GetControlledPawn (Utilities|Array|Get(acopy) (Fortnite|GetLocalFortPlayerControllers true true))) (Variables|Default|GetCueTag_EarthquakeTremors) (Ability|GameplayCue|MakeGameplayCueParameters :NormalizedMagnitude 0.0 :RawMagnitude 0.0 :EffectContext "()" :AggregatedSourceTags "(GameplayTags=)" :AggregatedTargetTags "(GameplayTags=)" :Location "0,0,0" :Normal "0,0,0" :EffectCauser self :GameplayEffectLevel 1 :AbilityLevel 1 :bReplicateLocationWhenUsingMinimalRepProxy false)))))

(fn ConstructionScript
  (|CallParentFunction))
