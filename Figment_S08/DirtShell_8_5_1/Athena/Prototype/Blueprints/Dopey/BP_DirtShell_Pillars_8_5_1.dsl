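; BP_DirtShell_Pillars_8_5_1 runs a six-option pillar vote among weapon and plane choices, using prop health as the tally.
; A fallback timer forces the vote, tracking the lowest prop health despite "HighestCount" naming.
; The winner is reported on mesh-network Root and Edge nodes; Edge also updates props once.
; It also broadcasts light visibility through the global event router, hides props and plays client earthquake cues on dig activation.
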(event Custom|TurnLightsOn
  (CallFunction|SetLightsVisible true))

(event Custom|TurnLightsOff
  (CallFunction|SetLightsVisible false))

(event Custom|SetLightsVisible
  (params (bNewVisibility :type Boolean)))

(event EventSetLightsVisibleParent
  (params (NewVisibility :type Boolean))
  (bind _returnvalue (GameplayEventRouter|GetGlobalEventRouter))
  (Utilities|IsValid _returnvalue
    (:Is%20Valid
      _)
    (:Is%20Not%20Valid
      (GameplayEventRouter|WaitForGlobalEventRouter 0
        (:OnEventRouterAvailable
          _))))
  (GameplayEventRouter|BroadcastStatefulEvent _returnvalue (Utilities|Struct|MakeFigmentActorsVisibilityStructure NewVisibility) 0)
  (for _array_element (Variables|Default|GetPropVisiblity)
    (Rendering|SetActorHiddenInGame _array_element (not NewVisibility))))

(event EventStartAutoCompleteTimer
  (bind _autocompletetimehotfix (Variables|Default|GetAutoCompleteTimeHotfix))
  (bind _returnvalue (ScalableFloat|GetValueatLevel _autocompletetimehotfix))
  (if (Math|Byte|Equal(Byte) (Variables|Default|GetVoteWinner) 255)
    (bind _returnvalue_2 (Utilities|Time|SetTimerbyEvent (delegate Custom|AutoCompleteVote) _returnvalue))
    (Variables|Default|SetAutoCompleteTimer _returnvalue_2)
    (Fortnite|FortShippingLog 0 (Utilities|String|Append "SNOWLOG AutoComplete Timer started: " (Utilities|String|ToString(Float) _returnvalue) " seconds"))
    (else
      (Fortnite|FortShippingLog 0 "SNOWLOG AutoComplete Winner failed. Winner was not set do default."))))

(event Custom|AutoCompleteVote
  (for (_array_element _array_index) (Variables|Default|GetDirtShellProps)
    (if (< (Variables|/Script/FortniteGame.BuildingPropBigHealth|GetBigHealth _array_element) (Variables|Default|GetHighestCount))
      (Variables|Default|SetHighestCount (Variables|/Script/FortniteGame.BuildingPropBigHealth|GetBigHealth _array_element))
      (Variables|Default|SetHighestCountIndex _array_index)))
  (CallFunction|ForceVote (Math|Conversions|ToByte(Integer) (Variables|Default|GetHighestCountIndex)))
  (Fortnite|FortShippingLog 0 "SNOWLOG Forced vote completion since haven't recieved a health update from mesh daddy in ages."))

(event EventBeginPlay
  (|CallParentFunction)
  (Networking|SwitchHasAuthority
    (:Remote
      (CallFunction|SetLightsVisible false))))

(event EventWinnerCalledParent
  (bind _votewinner (Variables|Default|GetVoteWinner))
  (bind _returnvalue (Utilities|Select :Index _votewinner :Option%200 "GW.Tac" :Option%201 "GW.Grappler" :Option%202 "GW.DrumGun" :Option%203 "GW.Bouncer" :Option%204 "Plane" :Option%205 "GW.Sword"))
  (Utilities|Time|ClearandInvalidateTimerbyHandle (Variables|Default|GetAutoCompleteTimer))
  (switch Utilities|FlowControl|Switch|SwitchonEMeshNetworkNodeType (MeshNetworkSubsystem|GetMeshNetworkNodeType (GameInstanceSubsystems|GetMeshNetworkSubsystem))
    (:Root
      (Default|CallOnPillarsConcludedParent _returnvalue))
    (:Edge
      (Utilities|FlowControl|DoOnce false
        (:Completed
          (CallFunction|EventVoteConcluded)
          (Default|CallOnPillarsConcludedParent _returnvalue)
          (for (_array_element _array_index) (Variables|Default|GetDirtShellProps)
            (bind _asbp_dirt_shell_prop_pillars_8_5_1 (Utilities|Casting|CastToBP_DirtShellProp_Pillars_8_5_1 _array_element))
            (Class|BPDirtShellPropPillars851|IsWinner _asbp_dirt_shell_prop_pillars_8_5_1 (== _array_index (Math|Conversions|ToInteger(Byte) (Variables|Default|GetVoteWinner))))
            (Fortnite|SetCanbeDamaged _array_element)))))))

(event Custom|EventVoteConcluded
  ;[1 1 1] Event on client (use remote) to spawn effects, audio etc. on vote concluded
  (Networking|SwitchHasAuthority))

(event Collision|EventActorBeginOverlap
  (params (OtherActor :type Actor%20Object%20Reference))
  (|CallParentFunction OtherActor) ; This node is disabled and will not be called.\nDrag off pins to build functionality.
)

(event EventTick
  (params (DeltaSeconds :type Float%20%28single-precision%29))
  (|CallParentFunction DeltaSeconds) ; This node is disabled and will not be called.\nDrag off pins to build functionality.
)

(event Custom|TestActivation)

(event Custom|TriggeredQuake
  (Networking|SwitchHasAuthority
    (:Remote
      (Development|PrintString "quake")
      (GameplayCue|ExecuteGameplayCue%28Burst%29Local (Pawn|GetControlledPawn (Utilities|Array|Get(acopy) (Fortnite|GetLocalFortPlayerControllers true true))) (Variables|Default|GetCueTag_EarthquakeTremors) (Ability|GameplayCue|MakeGameplayCueParameters :NormalizedMagnitude 0.0 :RawMagnitude 0.0 :EffectContext "()" :AggregatedSourceTags "(GameplayTags=)" :AggregatedTargetTags "(GameplayTags=)" :Location "0,0,0" :Normal "0,0,0" :EffectCauser self :GameplayEffectLevel 1 :AbilityLevel 1 :bReplicateLocationWhenUsingMinimalRepProxy false)))))

(fn OnRep_ActivatedOnDig
  (if true
    (CallFunction|TriggeredQuake)))

(fn ConstructionScript
  (|CallParentFunction)
  (CallFunction|TurnLightsOff))
