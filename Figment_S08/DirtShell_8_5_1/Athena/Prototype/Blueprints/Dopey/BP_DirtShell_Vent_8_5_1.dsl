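; Server-driven actor that opens vents once the two dirt mounds blocking them are destroyed, counting already-missing mounds as dead.
; With FirstSetupComplete set, it briefly raises ActivatedOnDig, then enables four vents in a delayed stagger; otherwise it enables only vent_2.
; The ActivatedOnDig RepNotify plays an earthquake tremor GameplayCue locally on remote clients.
; TestActivation duplicates the activation sequence without the death count.
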(event EventBeginPlay
  (bind _dirt_01 (Variables|Default|Getdirt_01))
  (bind _dirt_02 (Variables|Default|Getdirt_02))
  (|CallParentFunction)
  (Networking|SwitchHasAuthority
    (:Authority
      (Utilities|FlowControl|Sequence
        (:then_0
          ;[1 1 1] Check to see if the dirt mounds blocking the vent have already been destroyed
          (Utilities|IsValid _dirt_01
            (:Is%20Valid
              (EventDispatchers|BindEventtoOnDied _dirt_01 (EventDispatchers|CreateEvent :!event Custom|DeathEvent)))
            (:Is%20Not%20Valid
              (CallFunction|PropDied))))
        (:then_1
          (Utilities|IsValid _dirt_02
            (:Is%20Valid
              (EventDispatchers|BindEventtoOnDied _dirt_02 (EventDispatchers|CreateEvent :!event Custom|DeathEvent)))
            (:Is%20Not%20Valid
              (CallFunction|PropDied))))))))

(event Collision|EventActorBeginOverlap
  (params (OtherActor :type Actor%20Object%20Reference))
  (|CallParentFunction OtherActor) ; This node is disabled and will not be called.\nDrag off pins to build functionality.
)

(event EventTick
  (params (DeltaSeconds :type Float%20%28single-precision%29))
  (|CallParentFunction DeltaSeconds) ; This node is disabled and will not be called.\nDrag off pins to build functionality.
)

(event Custom|TestActivation
  (if (Variables|Default|GetFirstSetupComplete)
    (Utilities|FlowControl|Sequence
      (:then_0
        (Variables|Default|SetActivatedOnDig true)
        (Utilities|FlowControl|Delay 4.0)
        (Variables|Default|SetActivatedOnDig false))
      (:then_1
        (Utilities|FlowControl|Delay 5.0)
        (Class|BGADudeBroMiniOff|SetDisabled (Variables|Default|Getvent_1) false)
        (Utilities|FlowControl|Delay 1.0)
        (Class|BGADudeBroMiniOff|SetDisabled (Variables|Default|Getvent_2) false)
        (Utilities|FlowControl|Delay 1.0)
        (Class|BGADudeBroMiniOff|SetDisabled (Variables|Default|Getvent_3) false)
        (Utilities|FlowControl|Delay 1.0)
        (Class|BGADudeBroMiniOff|SetDisabled (Variables|Default|Getvent_4) false)))
    (else
      (Class|BGADudeBroMiniOff|SetDisabled (Variables|Default|Getvent_2) false))))

(event Custom|PropDied
  (Math|Integer|IncrementInt (Variables|Default|GetNumDead)
    (:%20%20
      (if (== (Variables|Default|GetNumDead) 2)
        (if (Variables|Default|GetFirstSetupComplete)
          (Utilities|FlowControl|Sequence
            (:then_0
              (Variables|Default|SetActivatedOnDig true)
              (Utilities|FlowControl|Delay 4.0)
              (Variables|Default|SetActivatedOnDig false))
            (:then_1
              (Utilities|FlowControl|Delay 5.0)
              (Class|BGADudeBroMiniOff|SetDisabled (Variables|Default|Getvent_1) false)
              (Utilities|FlowControl|Delay 1.0)
              (Class|BGADudeBroMiniOff|SetDisabled (Variables|Default|Getvent_2) false)
              (Utilities|FlowControl|Delay 1.0)
              (Class|BGADudeBroMiniOff|SetDisabled (Variables|Default|Getvent_3) false)
              (Utilities|FlowControl|Delay 1.0)
              (Class|BGADudeBroMiniOff|SetDisabled (Variables|Default|Getvent_4) false)))
          (else
            (Class|BGADudeBroMiniOff|SetDisabled (Variables|Default|Getvent_2) false)))))))

(event Custom|DeathEvent
  (params (DamagedActor :type Actor%20Object%20Reference) (Damage :type Float%20%28single-precision%29) (InstigatedBy :type Controller%20Object%20Reference) (DamageCauser :type Actor%20Object%20Reference) (HitLocation :type Vector) (FHitComponent :type Primitive%20Component%20Object%20Reference) (BoneName :type Name) (Momentum :type Vector))
  (CallFunction|PropDied))

(event Custom|TriggeredQuake
  (Networking|SwitchHasAuthority
    (:Remote
      (Development|PrintString "quake")
      (GameplayCue|ExecuteGameplayCue%28Burst%29Local (Pawn|GetControlledPawn (Utilities|Array|Get(acopy) (Fortnite|GetLocalFortPlayerControllers true true))) (Variables|Default|GetCueTag_EarthquakeTremors) (Ability|GameplayCue|MakeGameplayCueParameters :NormalizedMagnitude 0.0 :RawMagnitude 0.0 :EffectContext "()" :AggregatedSourceTags "(GameplayTags=)" :AggregatedTargetTags "(GameplayTags=)" :Location "0,0,0" :Normal "0,0,0" :EffectCauser self :GameplayEffectLevel 1 :AbilityLevel 1 :bReplicateLocationWhenUsingMinimalRepProxy false)))))

(fn OnRep_ActivatedOnDig
  (if (Variables|Default|GetActivatedOnDig)
    (CallFunction|TriggeredQuake)))

(fn ConstructionScript
  (|CallParentFunction))
