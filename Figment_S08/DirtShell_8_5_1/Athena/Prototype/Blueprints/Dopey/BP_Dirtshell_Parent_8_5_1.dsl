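; Parent Blueprint for a Dirtshell prototype actor with a replicated vote winner, apparently meant as a base for child variants.
; Its auto-complete timer, winner-called and light-visibility events are empty hooks, as are BeginPlay, overlap and Tick.
; The only logic is OnRep_VoteWinner, which tests the winner against sentinel 255 and contains a WinnerCalledParent call plus a mesh network Root/Edge switch with empty branches.
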(event Custom|StartAutoCompleteTimer)

(event Custom|WinnerCalledParent)

(event Custom|SetLightsVisibleParent
  (params (NewVisibility :type Boolean)))

(event EventBeginPlay)

(event Collision|EventActorBeginOverlap
  (params (OtherActor :type Actor%20Object%20Reference)))

(event EventTick
  (params (DeltaSeconds :type Float%20%28single-precision%29)))

(fn OnRep_VoteWinner
  (if (not (Math|Byte|Equal(Byte) (Variables|Default|GetVoteWinner) 255))
    (switch Utilities|FlowControl|Switch|SwitchonEMeshNetworkNodeType (MeshNetworkSubsystem|GetMeshNetworkNodeType (GameInstanceSubsystems|GetMeshNetworkSubsystem))
      (:Root
        _)
      (:Edge
        _))
    (CallFunction|WinnerCalledParent)))

(fn ConstructionScript)
